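# Physics asset for the skeletal mesh of the Tropical pickaxe, a cosmetic two-handed melee weapon.
# It holds a single capsule body on the Melee_TwoHanded_Handle bone, rotated to enclose the weapon as one simplified collision shape.
# There are no constraints, so it defines no articulated ragdoll or secondary motion between parts.
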PhysicsAsset /BRCosmetics/Weapons/FORT_Melee/Pickaxe_Tropical/Meshes/SK_Pickaxe_Tropical_PhysicsAsset
bodies {Melee_TwoHanded_Handle [{shapeType=Capsule, length=56.9835, radius=25.4894, center=<1.121 -0.8891 25.9443>, rotation=<rot 0.1252 -90 0.0351>}]}
constraints []
